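# Physics asset for the Sweet Teriyaki Yellow backpack cosmetic, giving capsule bodies and joints to a creature with head, eyes, fins and tail.
# The dyn_body bone is angularly locked to the pack bone and branches into front and tail chains.
# The front chain carries two-segment fins and a head with loosely limited eyes.
# The tail runs through tail7, which forks to tail5 and tail1-tail2, ending in tail3 and tail4, with mostly tight swing and twist limits.
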PhysicsAsset /BRCosmetics/Accessories/FORT_Backpacks/Backpack_M_MED_Sweet_Teriyaki_Yellow/Meshes/M_MED_Sweet_Teriyaki_Yellow_Pack_Physics
bodies {
 dyn_body [{shapeType=Capsule, length=0.505, radius=0.505, shapeName=dyn_body_capsule}]
 dyn_body_head [{shapeType=Capsule, length=3, radius=4, shapeName=dyn_body_head_capsule, center=<1 0 0>, rotation=<rot 0 0 90>}]
 dyn_body_tail [{shapeType=Capsule, length=4.5676, radius=3.0909, shapeName=dyn_body_tail_capsule, center=<1 0 -0.0001>, rotation=<rot 90 0 90>}]
 dyn_eye_l [{shapeType=Capsule, length=1, radius=1.5, shapeName=dyn_eye_l_capsule, center=<1 0 0>, rotation=<rot 90 0 0>}]
 dyn_eye_r [{shapeType=Capsule, length=1, radius=1.5, shapeName=dyn_eye_r_capsule, center=<1 0 0>, rotation=<rot -90 0 0>}]
 dyn_fin_l1 [{shapeType=Capsule, length=1.5, radius=1.5, shapeName=dyn_fin_l1_capsule, center=<1 0 0>, rotation=<rot -90 0 0>}]
 dyn_fin_l2 [{shapeType=Capsule, length=1.5, radius=1.5, shapeName=dyn_fin_l2_capsule, rotation=<rot -90 0 0>}]
 dyn_fin_r1 [{shapeType=Capsule, length=1.5, radius=1.5, shapeName=dyn_fin_r1_capsule, center=<1 0 0>, rotation=<rot 90 0 0>}]
 dyn_fin_r2 [{shapeType=Capsule, length=1.5, radius=1.5, shapeName=dyn_fin_r2_capsule, center=<1 0 0>, rotation=<rot 90 0 0>}]
 dyn_head [{shapeType=Capsule, length=4, radius=3, shapeName=dyn_head_capsule, center=<1 0 0>}]
 dyn_tail1 [{shapeType=Capsule, length=1.5, radius=1.5, shapeName=dyn_tail1_capsule, center=<1 0 0>, rotation=<rot 90 0 90>}]
 dyn_tail2 [{shapeType=Capsule, length=1.5, radius=1.5, shapeName=dyn_tail2_capsule, center=<1 0 0>, rotation=<rot 90 0 90>}]
 dyn_tail3 [{shapeType=Capsule, length=1, radius=1, shapeName=dyn_tail3_capsule, center=<1 0 0>, rotation=<rot 90 0 90>}]
 dyn_tail4 [{shapeType=Capsule, length=1, radius=1, shapeName=dyn_tail4_capsule, center=<1 0 0>, rotation=<rot 90 0 90>}]
 dyn_tail5 [{shapeType=Capsule, length=1, radius=1, shapeName=dyn_tail5_capsule, center=<1 0 0>, rotation=<rot 90 0 90>}]
 dyn_tail7 [{shapeType=Capsule, length=3.5915, radius=2.7042, shapeName=dyn_tail7_capsule, center=<1 0 0>, rotation=<rot 90 0 90>}]
 pack [{shapeType=Capsule, length=0.505, radius=0.6764, shapeName=pack_capsule, rotation=<rot 0 90 0>}]
}
constraints [
 {bone1Name=dyn_eye_r, bone2Name=dyn_head, swing1LimitDegrees=45, swing2LimitDegrees=45, twistLimitDegrees=45}
 {bone1Name=dyn_fin_r2, bone2Name=dyn_fin_r1, swing1LimitDegrees=20, swing2LimitDegrees=20, twistLimitDegrees=20}
 {bone1Name=dyn_body_tail, bone2Name=dyn_body, swing1LimitDegrees=10, swing2LimitDegrees=20, twistLimitDegrees=5}
 {bone1Name=dyn_tail7, bone2Name=dyn_body_tail, swing1LimitDegrees=10, swing2LimitDegrees=10, twistLimitDegrees=5}
 {bone1Name=dyn_tail1, bone2Name=dyn_tail7, swing1LimitDegrees=10, swing2LimitDegrees=10, twistLimitDegrees=5}
 {bone1Name=dyn_tail2, bone2Name=dyn_tail1, swing1LimitDegrees=10, swing2LimitDegrees=10, twistLimitDegrees=10}
 {bone1Name=dyn_tail4, bone2Name=dyn_tail2, swing1LimitDegrees=20, swing2LimitDegrees=20, twistLimitDegrees=1}
 {bone1Name=dyn_tail3, bone2Name=dyn_tail2, swing1LimitDegrees=20, swing2LimitDegrees=20, twistLimitDegrees=1}
 {bone1Name=dyn_tail5, bone2Name=dyn_tail7, swing1LimitDegrees=20, swing2LimitDegrees=20, twistLimitDegrees=1}
 {bone1Name=dyn_fin_l2, bone2Name=dyn_fin_l1, swing1LimitDegrees=20, swing2LimitDegrees=20, twistLimitDegrees=20}
 {
  bone1Name=dyn_body
  bone2Name=pack
  swing1LimitDegrees=45
  swing1Motion=Locked
  swing2LimitDegrees=45
  swing2Motion=Locked
  twistLimitDegrees=45
  twistMotion=Locked
 }
 {bone1Name=dyn_eye_l, bone2Name=dyn_head, swing1LimitDegrees=45, swing2LimitDegrees=45, twistLimitDegrees=45}
 {bone1Name=dyn_head, bone2Name=dyn_body_head, swing1LimitDegrees=20, swing2LimitDegrees=30, twistLimitDegrees=1}
 {bone1Name=dyn_fin_l1, bone2Name=dyn_body_head, swing1LimitDegrees=20, swing2LimitDegrees=20, twistLimitDegrees=20}
 {bone1Name=dyn_fin_r1, bone2Name=dyn_body_head, swing1LimitDegrees=20, swing2LimitDegrees=20, twistLimitDegrees=20}
 {bone1Name=dyn_body_head, bone2Name=dyn_body, swing1LimitDegrees=45, swing2LimitDegrees=5, twistLimitDegrees=1}
]
